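# Physics asset for the large male Order Guard cosmetic body, covering only its dangling gear: gun, knife and six left and right pockets.
# Each item pairs a small capsule on its base bone with a larger capsule on its 01 bone, linked by one constraint.
# Swing and twist limits keep motion tight on the gun and pockets, while the knife gets a wider range.
# The asset has no bodies for the main skeleton.
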PhysicsAsset /BRCosmetics/Characters/Player/Male/Large/Bodies/M_LRG_Order_Guard/Meshes/M_LRG_Order_Guard_Physics
bodies {
 dyn_gun_01 [{shapeType=Capsule, length=3, radius=3, shapeName=dyn_gun_01_capsule}]
 dyn_gun_base [{shapeType=Capsule, length=0.505, radius=0.505, shapeName=dyn_gun_base_capsule, rotation=<rot 0 -90 0>}]
 dyn_knife_01 [{shapeType=Capsule, length=3, radius=3, shapeName=dyn_knife_01_capsule}]
 dyn_knife_base [{shapeType=Capsule, length=0, radius=0.5, shapeName=dyn_knife_base_capsule, rotation=<rot 0 -90 0>}]
 dyn_pocket_A_l_01 [{shapeType=Capsule, length=3, radius=3, shapeName=dyn_pocket_A_l_01_capsule, rotation=<rot 0.0002 0 0>}]
 dyn_pocket_A_l_base [{shapeType=Capsule, length=0, radius=0.5, shapeName=dyn_pocket_A_l_base_capsule, rotation=<rot 0 -90 0>}]
 dyn_pocket_A_r_01 [{shapeType=Capsule, length=3, radius=3, shapeName=dyn_pocket_A_r_01_capsule}]
 dyn_pocket_A_r_base [{shapeType=Capsule, length=0, radius=0.5, shapeName=dyn_pocket_A_r_base_capsule, rotation=<rot 0 -90 0>}]
 dyn_pocket_B_l_01 [{shapeType=Capsule, length=3, radius=3, shapeName=dyn_pocket_B_l_01_capsule, rotation=<rot -0.0002 0 0>}]
 dyn_pocket_B_l_base [{shapeType=Capsule, length=0, radius=0.5, shapeName=dyn_pocket_B_l_base_capsule, rotation=<rot 0 -90 0>}]
 dyn_pocket_B_r_01 [{shapeType=Capsule, length=3, radius=3, shapeName=dyn_pocket_B_r_01_capsule, rotation=<rot -0.0002 0 0>}]
 dyn_pocket_B_r_base [{shapeType=Capsule, length=0, radius=0.5, shapeName=dyn_pocket_B_r_base_capsule, rotation=<rot 0 -90 0>}]
 dyn_pocket_C_l_01 [{shapeType=Capsule, length=3, radius=3, shapeName=dyn_pocket_C_l_01_capsule}]
 dyn_pocket_C_l_base [{shapeType=Capsule, length=0, radius=0.5, shapeName=dyn_pocket_C_l_base_capsule, rotation=<rot 0 -90 0>}]
 dyn_pocket_C_r_01 [{shapeType=Capsule, length=3, radius=3, shapeName=dyn_pocket_C_r_01_capsule}]
 dyn_pocket_C_r_base [{shapeType=Capsule, length=0, radius=0.5, shapeName=dyn_pocket_C_r_base_capsule, rotation=<rot 0 -90 0>}]
}
constraints [
 {bone1Name=dyn_pocket_C_r_01, bone2Name=dyn_pocket_C_r_base, swing1LimitDegrees=20, swing2LimitDegrees=20, twistLimitDegrees=20}
 {bone1Name=dyn_knife_01, bone2Name=dyn_knife_base, swing1LimitDegrees=45, swing2LimitDegrees=45, twistLimitDegrees=45}
 {bone1Name=dyn_pocket_A_r_01, bone2Name=dyn_pocket_A_r_base, swing1LimitDegrees=20, swing2LimitDegrees=20, twistLimitDegrees=20}
 {bone1Name=dyn_pocket_B_r_01, bone2Name=dyn_pocket_B_r_base, swing1LimitDegrees=20, swing2LimitDegrees=20, twistLimitDegrees=20}
 {bone1Name=dyn_pocket_A_l_01, bone2Name=dyn_pocket_A_l_base, swing1LimitDegrees=20, swing2LimitDegrees=20, twistLimitDegrees=20}
 {bone1Name=dyn_pocket_B_l_01, bone2Name=dyn_pocket_B_l_base, swing1LimitDegrees=20, swing2LimitDegrees=20, twistLimitDegrees=20}
 {bone1Name=dyn_pocket_C_l_01, bone2Name=dyn_pocket_C_l_base, swing1LimitDegrees=20, swing2LimitDegrees=20, twistLimitDegrees=20}
 {bone1Name=dyn_gun_01, bone2Name=dyn_gun_base, swing1LimitDegrees=20, swing2LimitDegrees=20, twistLimitDegrees=20}
]
