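# Post-process style AnimBlueprint for the male large Order Guard body, layering secondary deformation and cosmetic fixes onto the parent mesh's copied pose.
# Three Control Rigs feed shoulder-pad rotation constraints, low-alpha RigidBody physics, then additive pocket and pad offsets for crouch, parachute and skydive.
# When an OrderGuardMale-tagged pickaxe is equipped and unholstered, constraints and scaling collapse both hands and forearms.
# Update events convert those states into 0/1 alphas.
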AnimBlueprint /BRCosmetics/Characters/Player/Male/Large/Bodies/M_LRG_Order_Guard/Meshes/M_LRG_Order_Guard_AnimBP
target_skeleton=&/M_LRG_Order_Guard_Skeleton
graphs [
 AnimGraph {nodes [
   n1 = Control Rig(Source=Copy Pose From Mesh() {bUseAttachedParent=true}) {
    alphaBoolBlend {blendOption=Linear}
    controlRigAssetReference {blueprintRigClass=&/M_LRG_Order_Guard_Deform_CtrlRig_twistsB.M_LRG_Order_Guard_Deform_CtrlRig_twistsB_C}
   }
   n2 = Control Rig!(
    Source=@n1
    Alpha
    KneeRot_fr_l_114deg:Vector
    KneeTrans_fr_l_114deg:Vector="0.258,-3.591,0.009"
    KneeScale_fr_l_114deg:Vector="1,1,1"
    KneeScale_fr_l_130deg:Vector="1,1,1"
    KneeRot_fr_l_130deg:Vector="0,0,-51.2"
    KneeTrans_fr_l_130deg:Vector="0.258,-3.591,0.009"
   ) -> (Pose) {
    alphaBoolBlend {blendOption=Linear}
    controlRigAssetReference {blueprintRigClass=&/M_LRG_Order_Guard_Deform_CtrlRig_sculptersB.M_LRG_Order_Guard_Deform_CtrlRig_sculptersB_C}
   }
   n3 = Control Rig!(Source=@n2, Alpha) -> (Pose) {
    alphaBoolBlend {blendOption=Linear}
    controlRigAssetReference {blueprintRigClass=&/M_LRG_Order_Guard_CtrlRig.M_LRG_Order_Guard_CtrlRig_C}
   }
   n4 = "Constraint - dyn_shoulderpad_l_base (RR)"!(ComponentPose=Local To Component(LocalPose=@n3), Alpha=1, ConstraintWeights_0=0.5, ConstraintWeights_1:float=0.5) -> (Pose) {
    alphaBoolBlend {blendOption=Linear}
    boneToModify {boneName=dyn_shoulderpad_l_base}
    constraintSetup [{transformType=Rotation, targetBone {boneName=clavicle_l}}, {transformType=Rotation, targetBone {boneName=upperarm_l}}]
    constraintWeights [1,1]
   }
   n5 = "Constraint - dyn_shoulderpad_r_base (RR)"!(ComponentPose=@n4, Alpha=1, ConstraintWeights_0=0.5, ConstraintWeights_1:float=0.5) -> (Pose) {
    alphaBoolBlend {blendOption=Linear}
    boneToModify {boneName=dyn_shoulderpad_r_base}
    constraintSetup [{transformType=Rotation, targetBone {boneName=clavicle_r}}, {transformType=Rotation, targetBone {boneName=upperarm_r}}]
    constraintWeights [1,1]
   }
   n6 = RigidBody(ComponentPose=@n5, Alpha=0.1) {
    bOverrideWorldGravity=true
    overridePhysicsAsset=&/M_LRG_Order_Guard_Physics
    simulationSpace=BaseBoneSpace
    alphaBoolBlend {blendOption=Linear}
    baseBoneRef {boneName=root}
    componentAppliedLinearAccClamp=<1100 1100 1100>
    componentLinearAccScale=<0.5 0.5 0.5>
    componentLinearVelScale=<0.5 0.5 0.5>
    overrideWorldGravity {z=-980}
   }
   n7 = Get isCrouchingFloat() -> (isCrouchingFloat:double)
   n8 = "Transform (Modify) Bone - Bone: dyn_pocket_C_l_base"(ComponentPose=@n6, Alpha=@n7, Translation="0,3,0", Rotation="0,0,0", Scale="1,1,1") {
    rotationMode=BMM_Additive
    translationMode=BMM_Additive
    alphaBoolBlend {blendOption=Linear}
    boneToModify {boneName=dyn_pocket_C_l_base}
    translation {y=3}
   }
   n9 = "Transform (Modify) Bone - Bone: dyn_pocket_B_l_base"(ComponentPose=@n8, Alpha=@n7, Translation="-8,3,0", Rotation="-30,30,10", Scale="1,1,1") {
    rotationMode=BMM_Additive
    translationMode=BMM_Additive
    alphaBoolBlend {blendOption=Linear}
    boneToModify {boneName=dyn_pocket_B_l_base}
    rotation=<rot -30 30 10>
    translation=<-8 3>
   }
   n10 = Get isParachuteOpenFloat() -> (isParachuteOpenFloat:double)
   n11 = "Transform (Modify) Bone - Bone: dyn_shoulderpad_l_base"(ComponentPose=@n9, Alpha=@n10, Translation="1,0,0", Rotation="0,0,0", Scale="1,1,1") {
    rotationMode=BMM_Additive
    translationMode=BMM_Additive
    alphaBoolBlend {blendOption=Linear}
    boneToModify {boneName=dyn_shoulderpad_l_base}
    translation {x=1}
   }
   n12 = "Transform (Modify) Bone - Bone: dyn_shoulderpad_r_base"(ComponentPose=@n11, Alpha=@n10, Translation="-2,-2,-1", Rotation="20,0,0", Scale="1,1,1") {
    rotationMode=BMM_Additive
    translationMode=BMM_Additive
    alphaBoolBlend {blendOption=Linear}
    boneToModify {boneName=dyn_shoulderpad_r_base}
    rotation {pitch=20}
    translation=<-2 -2 -1>
   }
   n13 = Get isSkyDivingFloat() -> (isSkyDivingFloat:double)
   n14 = "Transform (Modify) Bone - Bone: dyn_shoulderpad_l_base"(ComponentPose=@n12, Alpha=@n13, Translation="0,0,3", Rotation="0,0,0", Scale="1,1,1") {
    rotationMode=BMM_Additive
    translationMode=BMM_Additive
    alphaBoolBlend {blendOption=Linear}
    boneToModify {boneName=dyn_shoulderpad_l_base}
    translation {z=3}
   }
   n15 = "Transform (Modify) Bone - Bone: dyn_shoulderpad_r_base"(ComponentPose=@n14, Alpha=@n13, Translation="0,0,0", Rotation="0,0,0", Scale="1,1,1") {rotationMode=BMM_Additive, translationMode=BMM_Additive, alphaBoolBlend {blendOption=Linear}, boneToModify {boneName=dyn_shoulderpad_r_base}}
   n16 = Get Order Guard Pickaxe() -> (Order Guard Pickaxe:double)
   n17 = "Constraint - upperarm_twist_02_r (P)"(ComponentPose=@n15, Alpha=@n16, ConstraintWeights_0=1) {
    alphaBoolBlend {blendOption=Linear}
    boneToModify {boneName=upperarm_twist_02_r}
    constraintSetup [{targetBone {boneName=lowerarm_twist_02_r}}]
    constraintWeights [1]
   }
   n18 = "Constraint - lowerarm_twist_01_r (P)"(ComponentPose=@n17, Alpha=@n16, ConstraintWeights_0=1) {
    alphaBoolBlend {blendOption=Linear}
    boneToModify {boneName=lowerarm_twist_01_r}
    constraintSetup [{targetBone {boneName=lowerarm_r}}]
    constraintWeights [1]
   }
   n19 = "Transform (Modify) Bone - Bone: hand_r"(ComponentPose=@n18, Alpha=@n16, Translation="0,0,0", Rotation="0,0,0", Scale="0.1,0.1,0.1") {scaleMode=BMM_Replace, scaleSpace=BCS_BoneSpace, alphaBoolBlend {blendOption=Linear}, boneToModify {boneName=hand_r}, scale=<0.1 0.1 0.1>}
   n20 = "Transform (Modify) Bone - Bone: lowerarm_twist_02_r"(ComponentPose=@n19, Alpha=@n16, Translation="0,0,0", Rotation="0,0,0", Scale="0.75,0.75,0.75") {
    scaleMode=BMM_Replace
    scaleSpace=BCS_BoneSpace
    alphaBoolBlend {blendOption=Linear}
    boneToModify {boneName=lowerarm_twist_02_r}
    scale=<0.75 0.75 0.75>
   }
   n21 = "Transform (Modify) Bone - Bone: lowerarm_twist_01_r"(ComponentPose=@n20, Alpha=@n16, Translation="0,0,0", Rotation="0,0,0", Scale="0.75,0.75,0.75") {
    scaleMode=BMM_Replace
    scaleSpace=BCS_BoneSpace
    alphaBoolBlend {blendOption=Linear}
    alphaScaleBiasClamp {bInterpResult=true}
    boneToModify {boneName=lowerarm_twist_01_r}
    scale=<0.75 0.75 0.75>
   }
   n22 = Get Order Guard Pickaxe()
   n23 = "Constraint - upperarm_twist_02_l (P)"(ComponentPose=@n21, Alpha=@n22, ConstraintWeights_0=1) {
    alphaBoolBlend {blendOption=Linear}
    boneToModify {boneName=upperarm_twist_02_l}
    constraintSetup [{targetBone {boneName=lowerarm_twist_02_l}}]
    constraintWeights [1]
   }
   n24 = "Constraint - lowerarm_twist_01_l (P)"(ComponentPose=@n23, Alpha=@n22, ConstraintWeights_0=1) {
    alphaBoolBlend {blendOption=Linear}
    boneToModify {boneName=lowerarm_twist_01_l}
    constraintSetup [{targetBone {boneName=lowerarm_l}}]
    constraintWeights [1]
   }
   n25 = "Transform (Modify) Bone - Bone: hand_l"(ComponentPose=@n24, Alpha=@n22, Translation="0,0,0", Rotation="0,0,0", Scale="0.1,0.1,0.1") {scaleMode=BMM_Replace, scaleSpace=BCS_BoneSpace, alphaBoolBlend {blendOption=Linear}, boneToModify {boneName=hand_l}, scale=<0.1 0.1 0.1>}
   n26 = "Transform (Modify) Bone - Bone: lowerarm_twist_02_l"(ComponentPose=@n25, Alpha=@n22, Translation="0,0,0", Rotation="0,0,0", Scale="0.75,0.75,0.75") {
    scaleMode=BMM_Replace
    scaleSpace=BCS_BoneSpace
    alphaBoolBlend {blendOption=Linear}
    boneToModify {boneName=lowerarm_twist_02_l}
    scale=<0.75 0.75 0.75>
   }
   n27 = "Transform (Modify) Bone - Bone: lowerarm_twist_01_l"(ComponentPose=@n26, Alpha=@n22, Translation="0,0,0", Rotation="0,0,0", Scale="0.75,0.75,0.75") {
    scaleMode=BMM_Replace
    scaleSpace=BCS_BoneSpace
    alphaBoolBlend {blendOption=Linear}
    alphaScaleBiasClamp {bInterpResult=true}
    boneToModify {boneName=lowerarm_twist_01_l}
    scale=<0.75 0.75 0.75>
   }
   Output Pose(Result=Component To Local(ComponentPose=@n27)) {layerGroup=None}
  ]}
 EventGraph {nodes [
   Event BlueprintUpdateAnimation()
   > n1 = Sequence!(execute) -> (then_0, then_1, then_2:Exec, then_3:Exec)
   > n2 = Branch(Condition=Get bIsCrouching())
   > Set isCrouchingFloat(isCrouchingFloat:double=1) -> (Output_Get:double)
   Set isCrouchingFloat(execute=@n2.else, isCrouchingFloat=0)
   n3 = Branch(execute=@n1.then_1, Condition=Get bIsParachuteOpened())
   > Set isParachuteOpenFloat(isParachuteOpenFloat:double=1) -> (Output_Get:double)
   Set isParachuteOpenFloat(execute=@n3.else, isParachuteOpenFloat=0)
   n4 = Branch(execute=@n1.then_2, Condition=Get bIsSkydiveFloating())
   > Set isSkyDivingFloat(isSkyDivingFloat:double=1) -> (Output_Get:double)
   Set isSkyDivingFloat(execute=@n4.else, isSkyDivingFloat=0)
   Set FortPawn(execute=@n1.then_3, FortPawn:Fort Pawn Object Reference=TryGetFortPawn()) -> (Output_Get:Fort Pawn Object Reference)
   n5 = Get FortPawn() -> (FortPawn:Fort Pawn Object Reference)
   > n6 = Is Valid(InputObject=@n5)
   n7 = Set weapon_visible(execute=@n6.Is Not Valid, weapon_visible:bool=false) -> (Output_Get:bool)
   n8 = Get weapon_visible() -> (weapon_visible:bool)
   n9 = Branch(execute=[@n13.then, @n7.then], Condition=@n8)
   n10 = PickaxeItemDefinitionHasAnyMetaTag(Tags=Make Array("[0]"=(TagName="Cosmetics.MetaTags.OrderGuardMale")), bRequiresEquipped=true)
   > Set Order Guard Pickaxe(Order Guard Pickaxe:double="To Float (Boolean)"(InBool=AND Boolean(A=@n8, B=@n10))) -> (Output_Get:double)
   Set Order Guard Pickaxe(execute=@n9.else, Order Guard Pickaxe=0)
   n11 = Get CurrentWeapon!(self:Fort Pawn Object Reference=@n5) -> (CurrentWeapon:Fort Weapon Object Reference)
   n12 = IsWeaponHolstered!(self:Fort Pawn Object Reference=@n5) -> (ReturnValue:bool)
   n13 = Set weapon_visible(execute=@n6.Is Valid, weapon_visible=AND Boolean(A=IsValid(Object=@n11), B=NOT Boolean(A=@n12)))
  ]}
]
variables [
 FortPawn {type=FortPawn, default=null}
 isCrouchingFloat {type=number}
 isParachuteOpenFloat {type=number}
 isSkyDivingFloat {type=number}
 Order Guard Pickaxe {type=number}
 weapon_visible {type=boolean}
]
